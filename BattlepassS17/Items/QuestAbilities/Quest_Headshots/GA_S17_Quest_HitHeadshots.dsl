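; Season 17 quest gameplay ability that reports headshot hits to the owning player's quest manager.
; ShouldAbilityRespondToEvent filters for critical hits, checking the target for a player pawn on another team or predator wildlife.
; On activation, a Blocked flag with a 0.2-second delay debounces rapid hits, so overlapping activations end without reporting.
; The first activation then sends a custom stat event with the headshot tag and target tags.
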(event Ability|EventActivateAbilityFromEvent
  (params (EventData :type Gameplay%20Event%20Data%20Structure%20%28by%20ref%29))
  (bind (_eventtag _instigator _target _optionalobject _optionalobject2 _contexthandle _instigatortags _targettags) (Utilities|Struct|BreakGameplayEventData EventData))
  (|CallParentFunction EventData)
  (if (Variables|Default|GetBlocked)
    (Utilities|FlowControl|Delay 0.2)
    (Ability|EndAbility)
    (else
      (Variables|Default|SetBlocked true)
      (Utilities|FlowControl|Delay 0.2)
      (Variables|Default|SetBlocked false)
      (Quest|SendComplexCustomStatEvent (Variables|GAVars|Data|GetOwningPlayerQuestManager) 0 (Variables|Default|GetT_Headshot) _targettags 1)
      (Ability|EndAbility))))

(fn ShouldAbilityRespondToEvent
  (params (ActorInfo :type Gameplay%20Ability%20Actor%20Info%20Structure) (Payload_EventTag :type Gameplay%20Tag%20Structure) (Payload_Instigator :type Actor%20Object%20Reference) (Payload_Target :type Actor%20Object%20Reference) (Payload_OptionalObject :type Object%20Reference) (Payload_OptionalObject2 :type Object%20Reference) (Payload_ContextHandle :type Gameplay%20Effect%20Context%20Handle%20Structure) (Payload_InstigatorTags :type Gameplay%20Tag%20Container%20Structure) (Payload_TargetTags :type Gameplay%20Tag%20Container%20Structure) (Payload_EventMagnitude :type Float%20%28single-precision%29) (Payload_TargetData :type Gameplay%20Ability%20Target%20Data%20Handle%20Structure) (Payload_InstancedEventData :type Instanced%20Struct%20Structure))
  (bind _returnvalue (Ability|EffectContext|IsCriticalHit Payload_ContextHandle))
  (bind _asfort_player_pawn (Utilities|Casting|CastToFortPlayerPawn Payload_Target)
    (:then
      (return (and _returnvalue (not (Team|OnSameTeam _asfort_player_pawn Payload_Instigator)))))
    (:CastFailed
      (Utilities|Casting|CastToNPC_Pawn_Wildlife_Predator_Parent Payload_Target)
      (if _returnvalue
        (return true)
        (else
          (return false))))))
